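; GAT_CreativePassiveAbility is a parent gameplay ability for Creative passive abilities that support property overrides.
; On activation it calls the parent, applies source property overrides, and stores the names of the overridden properties in a variable.
; IsPropertyOverridden reports whether a given property name is in that stored list.
; The end-ability event holds only a disabled parent call.
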(event Ability|EventActivateAbility
  (|CallParentFunction)
  (bind _overriddenpropertynames (Ability|ApplySourcePropertyOverrides))
  (Variables|Default|SetOverriddenPropertyNames _overriddenpropertynames))

(event Ability|EventOnEndAbility
  (params (bWasCancelled :type Boolean))
  (|CallParentFunction bWasCancelled) ; This node is disabled and will not be called.\nDrag off pins to build functionality.
)

(fn IsPropertyOverridden
  (params (PropertyName :type Name))
  (return (Utilities|Array|ContainsItem (Variables|Default|GetOverriddenPropertyNames) PropertyName)))
