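; Parent gameplay ability for Creative passive abilities that run on a Fort pawn.
; On activation it calls the parent, then stores the pawn returned by SetupPawnAbility as AbilityOwner.
; SetupPawnAbility casts the avatar actor from the ability's actor info to FortPawn, printing an error and ending the ability if the cast fails.
; The OnEndAbility event is only a disabled placeholder.
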(event Ability|EventActivateAbility
  (|CallParentFunction)
  (bind _fortpawn (CallFunction|SetupPawnAbility))
  (Variables|Default|SetAbilityOwner _fortpawn))

(event Ability|EventOnEndAbility
  (params (bWasCancelled :type Boolean))
  (|CallParentFunction bWasCancelled) ; This node is disabled and will not be called.\nDrag off pins to build functionality.
)

(fn SetupPawnAbility
  (bind (_owneractor _avataractor) (Utilities|Struct|BreakGameplayAbilityActorInfo (Ability|GetActorInfo)))
  (bind _asfort_pawn (Utilities|Casting|CastToFortPawn _avataractor)
    (:then
      (LocalVariables|SetLocalFortPawn _asfort_pawn))
    (:CastFailed
      (Development|PrintString "***CAST FAILED*** - Not a Pawn, Aborting")
      (Ability|EndAbility)))
  (return (LocalVariables|GetLocalFortPawn)))
