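; Parent gameplay ability for Creative active abilities, providing property override tracking and a timeout failsafe.
; On activation it calls its parent, applies source property overrides and stores the overridden property names for lookup through IsPropertyOverridden.
; SetupAbilityFailsafeTimer starts a timer for a given duration that ends the ability with an unexpected-timeout reason naming it.
; Nothing here starts or clears that timer, and the end-ability event is empty.
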(event Custom|SetupAbilityFailsafeTimer
  (params (Duration :type Float%20%28double-precision%29))
  (Utilities|Time|SetTimerbyEvent (delegate Custom|AbilityTimeout) Duration))

(event Custom|AbilityTimeout
  (CallFunction|EndAbilitywithReason (Utilities|String|Append "Ability " (Utilities|GetDisplayName self) " Ended due to unexpected timeout.")))

(event Ability|EventActivateAbility
  (|CallParentFunction)
  (bind _overriddenpropertynames (Ability|ApplySourcePropertyOverrides))
  (Variables|Default|SetOverriddenPropertyNames _overriddenpropertynames))

(event Ability|EventOnEndAbility
  (params (bWasCancelled :type Boolean)))

(fn IsPropertyOverridden
  (params (PropertyName :type Name))
  (return (Utilities|Array|ContainsItem (Variables|Default|GetOverriddenPropertyNames) PropertyName)))
