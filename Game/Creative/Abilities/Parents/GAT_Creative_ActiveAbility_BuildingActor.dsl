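; Parent gameplay ability for Creative active abilities that run on a Building Actor avatar.
; On activation it calls the parent implementation, then SetupBuildingActorAbility casts the avatar from the actor info to BuildingActor and stores the result as AbilityOwner.
; If the cast fails, it prints an error and calls EndAbility.
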(event Ability|EventActivateAbility
  (|CallParentFunction)
  (bind _buildingactor (CallFunction|SetupBuildingActorAbility))
  (Variables|Default|SetAbilityOwner _buildingactor))

(fn SetupBuildingActorAbility
  (bind (_owneractor _avataractor) (Utilities|Struct|BreakGameplayAbilityActorInfo (Ability|GetActorInfo)))
  (bind _asbuilding_actor (Utilities|Casting|CastToBuildingActor _avataractor)
    (:then
      (LocalVariables|SetLocalBuildingActor _asbuilding_actor))
    (:CastFailed
      (Development|PrintString "***CAST FAILED*** - Not a Building Actor,  Aborting")
      (Ability|EndAbility)))
  (return (LocalVariables|GetLocalBuildingActor)))
