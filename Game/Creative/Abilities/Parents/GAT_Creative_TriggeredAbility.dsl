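; Parent gameplay ability for Creative triggered abilities, providing property-override tracking and a failsafe timeout.
; On activation from an event it calls the parent, applies source property overrides and stores the overridden property names, which IsPropertyOverridden queries.
; SetupAbilityFailsafeTimer starts a timer for a given duration that ends the ability with an unexpected-timeout reason naming it.
; Activation here does not start that timer itself.
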(event Custom|SetupAbilityFailsafeTimer
  (params (Duration :type Float%20%28double-precision%29))
  (Utilities|Time|SetTimerbyEvent (delegate Custom|AbilityTimeout) Duration))

(event Custom|AbilityTimeout
  (CallFunction|EndAbilitywithReason (Utilities|String|Append "Ability " (Utilities|GetDisplayName self) " Ended due to unexpected timeout.")))

(event Ability|EventActivateAbilityFromEvent
  (params (EventData :type Gameplay%20Event%20Data%20Structure%20%28by%20ref%29))
  (|CallParentFunction EventData)
  (bind _overriddenpropertynames (Ability|ApplySourcePropertyOverrides))
  (Variables|Default|SetOverriddenPropertyNames _overriddenpropertynames))

(fn IsPropertyOverridden
  (params (PropertyName :type Name))
  (return (Utilities|Array|ContainsItem (Variables|Default|GetOverriddenPropertyNames) PropertyName)))
