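; Parent gameplay ability for Creative triggered abilities whose avatar is a Fort pawn.
; On event activation it calls the parent implementation, then SetupPawnActorAbility casts the avatar from the actor info to FortPawn and stores the result as AbilityOwner.
; A failed cast prints a debug message and ends the ability, though AbilityOwner is still assigned the returned local pawn variable.
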(event Ability|EventActivateAbilityFromEvent
  (params (EventData :type Gameplay%20Event%20Data%20Structure%20%28by%20ref%29))
  (|CallParentFunction EventData)
  (bind _fortpawn (CallFunction|SetupPawnActorAbility))
  (Variables|Default|SetAbilityOwner _fortpawn))

(fn SetupPawnActorAbility
  (bind (_owneractor _avataractor) (Utilities|Struct|BreakGameplayAbilityActorInfo (Ability|GetActorInfo)))
  (bind _asfort_pawn (Utilities|Casting|CastToFortPawn _avataractor)
    (:then
      (LocalVariables|SetLocalFortPawn _asfort_pawn))
    (:CastFailed
      (Development|PrintString "***CAST FAILED*** - Not a Fort Pawn, Aborting")
      (Ability|EndAbility)))
  (return (LocalVariables|GetLocalFortPawn)))
